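# Physics asset for the SweetCuddle Strawberry backpack cosmetic, giving its plush-like character jointed capsule bodies for secondary motion.
# Head, feet, tail and heart hands constrain to the body, while ears, nose and bottom hair hang off the head, with top hair chained to bottom hair.
# Every joint has a zero twist limit and its own swing limits, tightest at the head and looser on the small appendages.
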PhysicsAsset /BRCosmetics/Accessories/FORT_Backpacks/Backpack_F_MED_SweetCuddle_Strawberry/Meshes/F_MED_SweetCuddle_Strawberry_Pack_Physics
bodies {
 body [{shapeType=Capsule, length=3, radius=1, shapeName=body_capsule, rotation=<rot -0.0003 0 90>}]
 dyn_bottom_hair [{shapeType=Capsule, length=6.5, radius=0.5, shapeName=dyn_bottom_hair_capsule, center=<0.5451 0 0>}]
 dyn_ear_l [{shapeType=Capsule, length=1.5, radius=2, shapeName=dyn_ear_l_capsule, center=<3.3168 0 0>, rotation=<rot -90 -0.0001 0.0004>}]
 dyn_ear_r [{shapeType=Capsule, length=1.5, radius=2, shapeName=dyn_ear_r_capsule, center=<3.3168 0 0>, rotation=<rot -90 180 179.9997>}]
 dyn_foot_l [{shapeType=Capsule, length=1, radius=3, shapeName=dyn_foot_l_capsule, center=<3.643 0 0.6424>, rotation=<rot -80.0001 -0.0001 0.0001>}]
 dyn_foot_r [{shapeType=Capsule, length=1, radius=3, shapeName=dyn_foot_r_capsule, center=<3.643 0 0.6424>, rotation=<rot -80.0001 0 -0.0001>}]
 dyn_head [{shapeType=Capsule, length=5.5, radius=4.5, shapeName=dyn_head_capsule, center=<4.5538 -6.5031 0>, rotation=<rot 0 35.0021 -90>}]
 dyn_heart_hands [{shapeType=Capsule, length=2, radius=1.5, shapeName=dyn_heart_hands_capsule, center=<3.6796 0 0>}]
 dyn_nose [{shapeType=Capsule, length=2, radius=1, shapeName=dyn_nose_capsule, center=<1.1981 0 0>, rotation=<rot -0.0002 0 -90>}]
 dyn_tail [{shapeType=Capsule, length=0.5, radius=2, shapeName=dyn_tail_capsule, center=<2.3125 0 0>, rotation=<rot -90 0 0>}]
 dyn_top_hair [{shapeType=Capsule, length=4.5, radius=0.5, shapeName=dyn_top_hair_capsule, center=<0.5451 0 0>}]
}
constraints [
 {bone1Name=dyn_foot_r, bone2Name=body, swing1LimitDegrees=30, swing2LimitDegrees=30, twistLimitDegrees=0}
 {bone1Name=dyn_foot_l, bone2Name=body, swing1LimitDegrees=30, swing2LimitDegrees=30, twistLimitDegrees=0}
 {bone1Name=dyn_tail, bone2Name=body, swing1LimitDegrees=30, swing2LimitDegrees=30, twistLimitDegrees=0}
 {bone1Name=dyn_heart_hands, bone2Name=body, swing1LimitDegrees=45, swing2LimitDegrees=40, twistLimitDegrees=0}
 {bone1Name=dyn_head, bone2Name=body, swing1LimitDegrees=12, swing2LimitDegrees=12, twistLimitDegrees=0}
 {bone1Name=dyn_nose, bone2Name=dyn_head, swing1LimitDegrees=45, swing2LimitDegrees=45, twistLimitDegrees=0}
 {bone1Name=dyn_ear_r, bone2Name=dyn_head, swing1LimitDegrees=45, swing2LimitDegrees=30, twistLimitDegrees=0}
 {bone1Name=dyn_ear_l, bone2Name=dyn_head, swing1LimitDegrees=45, swing2LimitDegrees=30, twistLimitDegrees=0}
 {bone1Name=dyn_bottom_hair, bone2Name=dyn_head, swing1LimitDegrees=45, swing2LimitDegrees=45, twistLimitDegrees=0}
 {bone1Name=dyn_top_hair, bone2Name=dyn_bottom_hair, swing1LimitDegrees=45, swing2LimitDegrees=45, twistLimitDegrees=0}
]
